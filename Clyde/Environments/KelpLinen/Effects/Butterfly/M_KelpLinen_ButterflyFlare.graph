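# Additive unlit flare material for KelpLinen butterfly effects, drawing glowing camera-facing sprites that rotate about an axis.
# Emissive combines sphere-mask core and edge with a flare texture, tinted by per-instance hue shift.
# Opacity multiplies sine flicker with depth, camera and location fades.
# Sprite WPO, randomly sized per instance, passes through JuiceWPOMask before audio-reactive attributes; several branches, including rainbow and radial gradients, are unconnected.
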Material /Clyde/Environments/KelpLinen/Effects/Butterfly/M_KelpLinen_ButterflyFlare
graph {
 properties {
  bAllowNegativeEmissiveColor=true
  bUseMaterialAttributes=true
  bUsedWithInstancedStaticMeshes=true
  bUsedWithMeshParticles=true
  bUsedWithNanite=true
  bUsedWithNiagaraMeshParticles=true
  bUsedWithNiagaraSprites=true
  bUsedWithParticleSprites=true
  blendMode=BLEND_Additive
  pixelDepthOffsetMode=PDOM_Legacy
  shadingModel=MSM_Unlit
  twoSided=true
  usedShadingModels=Unlit
 }
 nodes [
  n1 = TransformPosition(None=WorldPosition().XYZ) {transformSourceType=TRANSFORMPOSSOURCE_World, transformType=TRANSFORMPOSSOURCE_Instance}
  n2 = Normalize(VectorInput=Subtract(A=CameraPositionWS(), B=ObjectPositionWS()))
  n3 = Normalize(VectorInput=CrossProduct(A=@n2, B=Constant3Vector() {constant=<rgba 0 0 1 0>}))
  n4 = MaterialFunctionCall(VectorToTransform=@n1, BasisX=@n2, BasisY=Multiply(A=@n3) {constB=-1}, BasisZ=Normalize(VectorInput=CrossProduct(A=@n3, B=@n2))) {materialFunction=&Transform3x3Matrix}
  n5 = Time()
  n6 = ObjectPositionWS()
  n7 = Add(A=@n5, B=ComponentMask(None=@n6) {b=true})
  n8 = ScalarParameter() {defaultValue=-0.2, parameterName=RotationSpeed}
  n9 = ScalarParameter() {defaultValue=-0.1, parameterName=RotationSpeed_Low}
  n10 = Frac(None=ComponentMask(None=ObjectPositionWS()) {b=true})
  n11 = RotateAboutAxis(
   NormalizedRotationAxis=Transform(None=VectorParameter() {parameterName=RotationAxis, defaultValue=<rgba 0 0 1 0>}.RGB) {transformSourceType=TRANSFORMSOURCE_Local}
   RotationAngle=Multiply(
    A=StaticSwitchParameter(True=@n7, False=@n5) {defaultValue=true, parameterName=UsewithGrassActor?}
    B=LinearInterpolate(A=@n8, B=@n9, Alpha=@n10)
   )
   PivotPoint=StaticSwitchParameter(True=@n6, False=ActorPositionWS()) {defaultValue=true, parameterName=UsewithGrassActor?}
   Position=WorldPosition().XYZ
  )
  Add(
   A=Multiply(
    A=Subtract(A=@n4.Transformed Vector, B=@n1)
    B=MaterialFunctionCall() {materialFunction=/Engine/Functions/Engine_MaterialFunctions02/WorldPositionOffset/ObjectScale}.Scale XYZ
   )
   B=@n11
  )
  n12 = TextureCoordinate()
  n13 = Constant() {r=0.5}
  n14 = VectorParameter() {parameterName=RadialGradientColorSetting, channelNames {a="", b=P, g=H, r=R}, defaultValue=<rgba 1 0 1 1>}
  LinearInterpolate(
   A=VectorParameter() {parameterName=EdgeColor, defaultValue=<rgba 1 0 0.6939 1>}.RGB
   B=VectorParameter() {parameterName=CenterColor, defaultValue=<rgba 0 0.8019 1 1>}.RGB
   Alpha=Power(Base=SphereMask(A=@n12, B=@n13, Radius=@n14.R, Hardness=@n14.R) {attenuationRadius=1, hardnessPercent=0}, Exp=@n14.P)
  )
  n15 = ComponentMask(None=TextureCoordinate() {uTiling=0, vTiling=0}) {g=true}
  n16 = Multiply(
   A=TextureSample(UVs=Panner(Coordinate=@n15) {speedX=1.6667}) {texture=/Game/Global/Textures/Default/Masks_and_Gradients/T_Grad_Rainbow}.RGB
   B=ParticleColor().RGB
  )
  MaterialFunctionCall(LDR Input=Clamp(None=@n16), HDR Tint=@n16) {materialFunction=&DeriveHDRfromLDR}
  n17 = Multiply(A=PerInstanceRandom(), B=ScalarParameter() {parameterName=HueShift})
  n18 = Constant3Vector() {constant=<rgba 0.5258 0 1 1>}
  n19 = DynamicParameter() {paramNames [FlickerFade,FlickerSpeed,ColorJitter,Param4]}
  n20 = Constant() {r=0.5}
  n21 = SphereMask(A=TextureCoordinate(), B=@n20, Radius=@n20) {hardnessPercent=0}
  n22 = ScalarParameter() {defaultValue=25, parameterName=CentertPower}
  n23 = Power(Base=@n21, Exp=ScalarParameter() {defaultValue=4, parameterName=EdgePower}) {constExponent=4}
  n24 = ScalarParameter() {defaultValue=0.15, parameterName=EdgeBrightness}
  n25 = Add(A=TextureCoordinate(), B=ComponentMask(None=VectorParameter() {parameterName=FlareOffset, defaultValue=<rgba 0 0 0 1>}.RGB) {g=true, r=true})
  n26 = Add(
   A=Multiply(A=ScalarParameter() {defaultValue=2, parameterName=CenterBrightness}, B=Power(Base=@n21, Exp=@n22) {constExponent=25}) {constB=2}
   B=Add(
    A=Multiply(A=@n23, B=@n24) {constB=0.15}
    B=TextureSampleParameter2D(UVs=@n25) {parameterName=FlareMask, texture=/Game/Effects/Fort_Effects/Textures/Flares/T_Flare_Spikeball}.RGB
   )
  )
  n27 = StaticSwitchParameter(
   True=MaterialFunctionCall(Hue Shift Percentage=@n19.ColorJitter, Texture=@n26) {materialFunction=&HueShift}.Result
   False=@n26
  ) {parameterName=UseColorJitter?}
  n28 = StaticSwitchParameter(
   True=Multiply(
    A=MaterialFunctionCall(Hue Shift Percentage=@n17, Texture=@n18) {materialFunction=&HueShift}.Result
    B=@n27
   )
   False=@n27
  ) {defaultValue=true, parameterName=UseRadialGradientColor}
  n29 = MaterialFunctionCall("FadeSensitivity (optional)"=ScalarParameter() {defaultValue=1, parameterName=AngleFade}) {materialFunction=&FadeBasedOnViewAngle}
  n30 = Constant() {r=1}
  n31 = PerInstanceRandom()
  n32 = Multiply(
   A=Time()
   B=StaticSwitchParameter(True=LinearInterpolate(Alpha=@n31) {constA=1.2, constB=0.8}, False=@n19.FlickerSpeed) {parameterName=GPU Particles}
  )
  n33 = Sine(None=@n32) {period=0.075}
  n34 = Multiply(A=Multiply(A=Sine(None=@n32) {period=0.155}, B=Sine(None=@n32) {period=0.2}), B=Sine(None=@n32) {period=0.43})
  n35 = ScalarParameter() {defaultValue=0.4, desc="0 = full flicker amount", parameterName=GPUFlickerFadeMin}
  n36 = ScalarParameter() {defaultValue=1, desc="1 = no flicker", parameterName=GPUFlickerFadeMax, sortPriority=1}
  n37 = PerInstanceRandom()
  n38 = LinearInterpolate(
   A=Clamp(None=Multiply(A=@n33, B=@n34)) {maxDefault=3}
   Alpha=StaticSwitchParameter(True=LinearInterpolate(A=@n35, B=@n36, Alpha=@n37) {constA=0.5}, False=Saturate(None=@n19.FlickerFade)) {parameterName=GPU Particles}
  ) {constA=1}
  n39 = Constant() {r=1}
  n40 = DynamicParameter() {desc=Index 1 Params for Niagara, parameterIndex=1, paramNames [DepthFadeDist, Param2, Param3, Param4]}
  n41 = ScalarParameter() {defaultValue=30, parameterName=DepthFade}
  n42 = If(A=@n40.DepthFadeDist, B=Constant() {r=1}, "A > B"=@n40.DepthFadeDist, "A == B"=@n41, "A < B"=@n41)
  n43 = Multiply(
   A=Multiply(
    A=StaticSwitchParameter(True=@n29.Result, False=@n30) {parameterName=Use Angle Fade}
    B=StaticSwitchParameter(True=@n38, False=@n39) {defaultValue=true, parameterName=UseFlicker}
   )
   B=ShadingPathSwitch(Default=DepthFade(FadeDistance=@n42))
  )
  n44 = ScalarParameter() {parameterName=CameraFadeDistance}
  n45 = StaticSwitchParameter(
   True=Multiply(A=@n43, B=MaterialFunctionCall(Distance=@n44) {materialFunction=&Near_Camera_Fade}.Result)
   False=@n43
  ) {group=DetailedGradientMask, parameterName=UseCameraNearFade}
  n46 = ScalarParameter() {defaultValue=500, parameterName=CamerFadeLength}
  n47 = ScalarParameter() {defaultValue=2500, parameterName=CameraFade_Offset}
  n48 = Multiply(
   A=Multiply(A=@n45, B=MaterialFunctionCall() {materialFunction=/ClydeCosmetics/Environments/KelpLinen/MF_KelpLinen_LocationFade}.Output)
   B=Clamp(None=MaterialFunctionCall(Fade Length=@n46, Fade Offset=@n47) {materialFunction=&CameraDepthFade}.Result)
  )
  n49 = ScalarParameter() {defaultValue=0.05, parameterName=Master_Opacity}
  n50 = ScalarParameter() {defaultValue=-5000, parameterName=JuiceWPOMask_Amount}
  n51 = MaterialFunctionCall(Vector 3=Transform(None=Constant3Vector() {constant=<rgba 1 0 0 0>}) {transformSourceType=TRANSFORMSOURCE_Instance}) {materialFunction=/Engine/Functions/Engine_MaterialFunctions02/Utility/VectorLength}
  n52 = MaterialFunctionCall(Vector 3=Transform(None=Constant3Vector() {constant=<rgba 0 1 0 0>}) {transformSourceType=TRANSFORMSOURCE_Instance}) {materialFunction=&VectorLength}
  n53 = ScalarParameter() {defaultValue=400, parameterName=Sprite_size_Min}
  n54 = ScalarParameter() {defaultValue=600, parameterName=Sprite_size_Max}
  n55 = PerInstanceRandom()
  n56 = MaterialFunctionCall!(
   Center Location=Add(A=@n11, B=WorldPosition().XYZ)
   0-1 UVs
   XY scale=Multiply(A=AppendVector(A=@n51.V3 Length, B=@n52.V3 Length), B=LinearInterpolate(A=@n53, B=@n54, Alpha=@n55))
   OverridePivot
   "Normals (Optional)"
  ) {materialFunction=&Sprite}
  MaterialFunctionCall!(In=SetMaterialAttributes!(
    MaterialAttributes
    Emissive Color=MaterialFunctionCall(Emissive=@n28) {materialFunction=&MF_FX_EmissiveAdjustments}.Result
    Opacity=Multiply(A=@n48, B=@n49)
    World Position Offset=MaterialFunctionCall!(WPO_Amount=@n50, Position, WPO=@n56.World position offset) {materialFunction=/Clyde/FrontierContent/Materials/MF_JuiceWPOMask}.Result
   ) {attributeSetTypes [#1, #2, #3]}) {materialFunction=/ClydeCosmetics/Environments/KelpLinen/MF_Kelp_AudioReactive}
  Multiply(
   A=@n45
   B=MaterialFunctionCall("FadeSensitivity (optional)"=ScalarParameter() {parameterName="FadeSensitivity (optional) (S)"}) {materialFunction=&FadeBasedOnViewAngle}.Result
  )
  ParticleRandom()
  ParticleRandom()
  Transform(None=ObjectPositionWS()) {transformSourceType=TRANSFORMSOURCE_World, transformType=TRANSFORM_Instance}
  VertexColor()
 ]
}
