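# Unlit translucent, two-sided material for a glowing butterfly aura on Niagara mesh particles.
# Two panning smoke samples and a radial blend of BaseColor to CrackColor form the emissive, brightened by Fresnel and tinted by particle color.
# A butterfly mask, smoke and particle alpha shape opacity.
# Wings flap by rotating about particle direction, signed by vertex red and driven by OuterFlapAnim; both sphere-masked lerp inputs use it, leaving FlapAnim unused.
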Material /Clyde/Environments/KelpLinen/Effects/Butterfly/Materials/M_KelpLinen_Aura
graph {
 properties {
  bUsedWithNiagaraMeshParticles=true
  blendMode=BLEND_Translucent
  pixelDepthOffsetMode=PDOM_Legacy
  previewMesh=/Clyde/Environments/KelpLinen/Effects/Butterfly/Meshes/SM_KelpLinen_Aura
  shadingModel=MSM_Unlit
  twoSided=true
  usedShadingModels=Unlit
 }
 nodes [
  n1 = ParticleColor()
  n2 = LinearInterpolate(Alpha=Fresnel() {baseReflectFraction=0, exponent=3}) {constA=0.7, constB=2}
  n3 = TextureCoordinate() {uTiling=3, vTiling=1.4}
  n4 = ScalarParameter() {defaultValue=1, parameterName=Opacity}
  n5 = Multiply(A=Subtract(A=Time(), B=Multiply(A=@n4) {constB=2}), B=ScalarParameter() {defaultValue=2, parameterName=PanSpeed})
  n6 = TextureCoordinate() {uTiling=1.5, vTiling=0.7}
  n7 = Add(
   A=TextureSample(UVs=Panner(Coordinate=@n3, Time=@n5) {speedX=-0.04, speedY=0.15}) {texture=&T_Inky_Smoke_Tile_Chromatic}.RGB
   B=TextureSample(UVs=Panner(Coordinate=@n6, Time=@n5) {speedX=0.05, speedY=0.25}) {texture=&T_Inky_Smoke_Tile_Chromatic}.RGB
  )
  n8 = VectorParameter() {parameterName=CrackColor, defaultValue=<rgba 0.1905 0.4686 0.88 0>}
  n9 = SphereMask(A=TextureCoordinate() {coordinateIndex=1}, B=Constant2Vector() {g=0.5, r=1}) {attenuationRadius=0.5, hardnessPercent=0}
  n10 = Add(
   A=Multiply(A=@n7, B=Multiply(A=@n8.RGB) {constB=1.5})
   B=LinearInterpolate(A=VectorParameter() {parameterName=BaseColor, defaultValue=<rgba 0.1137 0.1382 0.37 0>}.RGB, B=@n8.RGB, Alpha=Power(Base=@n9))
  )
  MaterialFunctionCall(Emissive=Multiply(A=@n1.RGB, B=Multiply(A=@n2, B=@n10))) {materialFunction=&MF_FX_EmissiveAdjustments}
  n11 = Add(A=Multiply(A=@n7) {constB=0.75}) {constB=0.7}
  n12 = TextureSample(UVs=TextureCoordinate() {coordinateIndex=1, uTiling=2}) {texture=/Game/Athena/Prototype/Blueprints/Island/T_Butterfly_M}
  Multiply(A=@n1.A, B=Multiply(A=@n4, B=Multiply(A=@n11, B=@n12.R)))
  n13 = VertexColor()
  n14 = ScalarParameter() {defaultValue=-0.26, parameterName=FlapMin}
  n15 = ScalarParameter() {defaultValue=0.25, parameterName=FlapMax}
  n16 = DynamicParameter() {paramNames [FlapAnim, OuterFlapAnim, Param3, Param4]}
  n17 = TextureCoordinate() {coordinateIndex=1}
  n18 = Constant2Vector() {g=0.15, r=1}
  n19 = ScalarParameter() {defaultValue=1, parameterName=SphereRadius}
  n20 = LinearInterpolate(
   A=@n16.OuterFlapAnim
   B=@n16.OuterFlapAnim
   Alpha=Power(Base=SphereMask(A=@n17, B=@n18, Radius=@n19) {attenuationRadius=0.4, hardnessPercent=0}) {constExponent=1.25}
  )
  RotateAboutAxis(
   NormalizedRotationAxis=ParticleDirection()
   RotationAngle=Multiply(A=LinearInterpolate(Alpha=@n13.R) {constA=-1}, B=LinearInterpolate(A=@n14, B=@n15, Alpha=@n20))
   PivotPoint=ParticlePositionWS()
   Position=WorldPosition().XYZ
  )
 ]
}
