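# Translucent unlit material for cracked-glass shard mesh particles, in the KelpLinen Butterfly effect folder.
# Two scales of crack normals distort a tinted, Fresnel-weighted cubemap reflection, added to a particle-colored glow from base color and crack masks.
# The glow passes through DeriveHDRfromLDR, or a plain intensity multiply on Low quality, before MF_FX_EmissiveAdjustments.
# Opacity combines Fresnel and crack mask over a minimum, scaled by particle alpha.
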Material /Clyde/Environments/KelpLinen/Effects/Butterfly/Materials/M_KelpLinen_Shard
graph {
 properties {
  bUsedWithMeshParticles=true
  bUsedWithNiagaraMeshParticles=true
  blendMode=BLEND_Translucent
  pixelDepthOffsetMode=PDOM_Legacy
  shadingModel=MSM_Unlit
  usedShadingModels=Unlit
 }
 nodes [
  Add(A=TextureCoordinate() {coordinateIndex=2, uTiling=1.25, vTiling=0.5}, B=Multiply(A=TextureCoordinate() {coordinateIndex=1}) {constB=0.5})
  n1 = TextureCoordinate() {uTiling=2, vTiling=1.6}
  n2 = Add(
   A=Multiply(
    A=TextureSampleParameter2D(UVs=@n1) {parameterName=NormalsTile, samplerType=SAMPLERTYPE_Normal, texture=/Game/Effects/Fort_Effects/Textures/Tile/T_Glass_Cracks_N}.RGB
    B=Constant3Vector() {constant=<rgba 1 1 0 0>}
   )
   B=Multiply(
    A=TextureSampleParameter2D() {parameterName=Normals, samplerType=SAMPLERTYPE_Normal, texture=&T_Glass_Cracks_N}.RGB
    B=Constant3Vector() {constant=<rgba 2 2 1 0>}
   )
  )
  n3 = TextureSampleParameterCube!(UVs=ReflectionVectorWS(CustomWorldNormal=Transform(None=@n2)) {bNormalizeCustomWorldNormal=true}, Apply View MipBias) {
   parameterName=Cubemap
   samplerType=SAMPLERTYPE_LinearColor
   texture=/Game/Packages/Fortress_Sky/SkyLightCaptures/Textures/T_SkyLightCapture_12
  }
  n4 = Multiply(A=Multiply(A=@n3.RGB, B=@n3.RGB), B=VectorParameter() {parameterName=CubemapTint, defaultValue=<rgba 2.2812 2.8289 3 1>}.RGB) {constB=0.7}
  n5 = Fresnel() {baseReflectFraction=0.4, exponent=1.5}
  n6 = ParticleColor()
  n7 = Fresnel() {baseReflectFraction=0, exponent=3}
  n8 = Add(
   A=Multiply(A=TextureSample(UVs=@n1) {texture=/Game/Effects/Fort_Effects/Textures/Tile/T_Glass_Cracks_M}.G) {constB=0.6}
   B=TextureSample() {texture=&T_Glass_Cracks_M}.G
  )
  n9 = Multiply(A=@n8, B=Multiply(A=VectorParameter() {parameterName=CrackColor, defaultValue=<rgba 0.231 0.8902 1 1>}.RGB) {constB=8})
  n10 = VectorParameter() {parameterName=BaseColor, defaultValue=<rgba 0.1035 0.2124 0.4115 1>}
  n11 = Multiply(A=@n6.RGB, B=Multiply(A=LinearInterpolate(Alpha=@n7) {constA=0.2}, B=Add(A=@n9, B=@n10.RGB)))
  n12 = ScalarParameter() {defaultValue=1.5, parameterName=HDRIntensity}
  n13 = MaterialFunctionCall(LDR Input=Saturate(None=@n11), HDR Tint=@n11, HDR Intensity Multiplier=@n12, Desaturate Luminance=Constant()) {materialFunction=&DeriveHDRfromLDR}
  n14 = Multiply(A=@n11, B=@n12)
  MaterialFunctionCall(Emissive=Add(A=Multiply(A=@n4, B=@n5), B=QualitySwitch(Default=@n13.RemappedRange, Low=@n14))) {materialFunction=&MF_FX_EmissiveAdjustments}
  n15 = ScalarParameter() {defaultValue=0.45, parameterName=MinOpacity}
  Multiply(A=@n6.A, B=Add(A=@n7, B=LinearInterpolate(A=@n15, Alpha=@n8) {constA=0.25}))
 ]
}
