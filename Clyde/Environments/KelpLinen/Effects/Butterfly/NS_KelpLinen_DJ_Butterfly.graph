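# Niagara effect of a glowing butterfly that assembles from swirling mesh shards, with ambient drifting shards and sparkles.
# ButterflyBody wanders around its spawn point and animates a flap; ButterflyShards reads its transform, blending 16 mirrored shards from noisy swirl into flapping wings via System.FormUp.
# System.Fling reverses the shards' vortex pull and agitates Sparkles, which spawn from the body.
# User.Fade scales wing-shard alpha; the Layout emitter is disabled.
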NiagaraSystem /Clyde/Environments/KelpLinen/Effects/Butterfly/NS_KelpLinen_DJ_Butterfly
dependencies {
 usedDataInterfaces [&NiagaraDataInterfaceCurve, &NiagaraDataInterfaceArrayFloat3, &NiagaraDataInterfaceParticleRead]
 usedDynamicInputs [&LerpPosition]
 usedModules [
  &SystemState
  &self:SystemSpawnScript.SystemScriptSource.SystemScriptGraph.NiagaraNodeAssignment_1.SetVariables_#1
  &self:SystemSpawnScript.SystemScriptSource.SystemScriptGraph.NiagaraNodeAssignment_0.SetVariables_#2
 ]
}
emitters {
 ButterflyBody {
  emitterUpdateScript [
   &EmitterState(
    Inactive Response=::NewEnumerator1
    Life Cycle Mode=::NewEnumerator1
    Loop Behavior=::NewEnumerator1
    Loop Duration Mode=::NewEnumerator1
    UseLoopDelay=false
   )
   &SpawnBurst_Instantaneous(Use Spawn Probability=false, UseLoopCountLimit=false)
  ]
  particleSpawnScript [
   &InitializeParticle(
    Color=<rgba 0.3473 1.2568 2.5 1>
    Lifetime=5
    Mesh Scale Mode=::NewEnumerator3
    Position Offset=<0 0 50>
    Sprite Size=<70 125>
    Sprite Size Mode=::NewEnumerator4
    UsePositionOffset=true
    Write Material Random=true
   )
   &AddVelocity(Velocity=<1 0 -0.5>, Velocity Speed Scale=50)
   /SetVariables(Particles.MiddleSpriteSize:Vector2f=&Vector2DFromFloat(Value=100)) {isSetParameters=true}
  ]
  particleUpdateScript [
   &ParticleState(Kill Particles When Lifetime Has Elapsed=false)
   &UpdateMeshOrientation(
    Facing Direction=$Particles.Velocity
    Facing Interpolation=false
    Orientation Method=::NewEnumerator1
    Override Quat=false
    TransformFacing=false
    TransformUp=false
   )
   /SetVariables(Particles.DistanceFromCenter:NiagaraFloat=&FloatFromCurve(
     FloatCurve=DataInterface() {propertyValues {Curve {keys [
         <at 0 0 RCIM_Cubic>
         {arriveTangentWeight=13.3334, interpMode=RCIM_Cubic, leaveTangentWeight=120.0004, tangentMode=RCTM_User, time=40, value=0}
         <at 300 1 RCIM_Cubic>
        ]}}}
     CurveIndex=&DistanceBetweenPositions(A=$Particles.Position, B=$Particles.Initial.Position)
    )) {isSetParameters=true}
   &VortexForce(
    Origin Pull Amount:NiagaraFloat=&Lerp_Float(A=-50, B=200, Alpha=$Particles.DistanceFromCenter)
    Origin Pull Only On Vortex Axis=false
    UseOriginOffset=false
    Vortex Force Amount=250
    Vortex Origin:NiagaraPosition=$Particles.Initial.Position
   )
   &CurlNoiseForce(
    Noise Frequency=10
    Noise Strength=&Waveform(Global Amplitude Scale=250, "Waveform [1]"=::NewEnumerator0, "[1] Frequency"=1, "[1] Amplitude Min/Max"=<-0.5 1>)
    Pan Noise=true
    Pan Noise Field=<0 0 1>
   )
   &Drag(Drag=2, Use Linear Drag=true, Use Rotational Drag=false)
   &SolveForcesAndVelocity(Acceleration Limit=300, Clamp Velocity=true, Limit Acceleration=true, Speed Limit=400)
   /Niagara/Modules/Update/Utility/IncrementOverTime(Rate of Change:NiagaraFloat=&Lerp_Float(A=0.1, B=1.25, Alpha=$System.FormUp))
   /SetVariables(Particles.FlapAnim:NiagaraFloat=&FloatFromCurve(
     FloatCurve=DataInterface() {propertyValues {Curve {keys [
         {interpMode=RCIM_Cubic, tangentMode=RCTM_User, time=0, value=0}
         {arriveTangent=4.1706, interpMode=RCIM_Cubic, leaveTangent=4.1706, tangentMode=RCTM_User, time=0.25, value=0.7}
         {
          arriveTangent=-0.2246
          arriveTangentWeight=0.2294
          interpMode=RCIM_Cubic
          leaveTangent=-0.2246
          leaveTangentWeight=0.0573
          tangentMode=RCTM_User
          time=0.45
          value=1
         }
         {arriveTangent=-2.436, interpMode=RCIM_Cubic, leaveTangent=-2.436, tangentMode=RCTM_User, time=0.6, value=0.4121}
         {tangentMode=RCTM_User, time=1, value=0}
        ]}}}
     CurveIndex=&ModuloFloat(A=$Particles.IncrementOverTime.Counter, B=1)
    )) {isSetParameters=true}
   &InterpolateOverTime(Rate Of Change=5, Target Float Value:NiagaraFloat=$Particles.FlapAnim)
   &DynamicMaterialParameters(
    Index 0 Param 1=$Particles.FlapAnim
    Index 0 Param 2=$Particles.InterpolateOverTime.InterpolateOverTime_Float.Moving Average
    Index 1 Param 1=0.2
    Index 1 Param 2=&RandomRangeFloat(Minimum=0.7, Maximum=1)
    Module.Index0Param0Write=true
    Module.Index0Param1Write=true
    Module.Index0Param2Write=false
    Module.Index0Param3Write=false
    Module.Index1Param0Write=true
    Module.Index1Param1Write=true
    Module.Index1Param2Write=false
    Module.Index1Param3Write=false
    Write Parameter Index 1=true
   )
   &ScaleColor(
    Scale Alpha=&FloatFromCurve(
     FloatCurve=DataInterface() {propertyValues {Curve {keys [<at 0.4 0 RCIM_Cubic>, {tangentMode=RCTM_User, time=0.8, value=1}]}}}
     CurveIndex=$System.FormUp
    )
    ScaleA=true
    ScaleRGB=false
   )
   /SetVariables(Particles.AuraColor:LinearColor=&MultiplyLinearColor(
     Linear Color=<rgba 0.5375 0.771 1.2 1>
     Scale Factor=&MakeFloat4FromVectorAndFloat(
      "Vector (RGB)"=<1 1 1>
      "Float (Alpha)"=&MakeFloatFromLinearColor(LinearColor=$Particles.Color, Channel:ENiagaraLinearColor_Channels=::NewEnumerator3)
     )
    )) {isSetParameters=true}
   /SetVariables(Particles.VisibilityTag=&MakeCustomIntFromBool(
     Bool=&SetBoolByFloatComparison(
      A=&MakeFloatFromLinearColor(LinearColor=$Particles.Color, Channel:ENiagaraLinearColor_Channels=::NewEnumerator3)
      B=0
     )
     True INT=0
     False INT=1
    )) {isSetParameters=true}
  ]
  rendererClasses [&NiagaraSpriteRendererProperties, &NiagaraMeshRendererProperties]
  renderers [
   {rendererClass=&NiagaraSpriteRendererProperties}
   {rendererClass=&NiagaraMeshRendererProperties, rendererIndex=1}
   {rendererClass=&NiagaraSpriteRendererProperties, rendererIndex=2}
   {rendererClass=&NiagaraSpriteRendererProperties, rendererIndex=3}
  ]
 }
 ButterflyShards {
  emitterUpdateScript [
   &EmitterState(
    Inactive Response=::NewEnumerator1
    Life Cycle Mode=::NewEnumerator1
    Loop Behavior=::NewEnumerator1
    Loop Duration Mode=::NewEnumerator1
    UseLoopDelay=false
   )
   &SpawnRate(Use Spawn Probability=false) {enabled=false}
   &SpawnBurst_Instantaneous(Spawn Count=16, Use Spawn Probability=false, UseLoopCountLimit=false)
  ]
  particleSpawnScript [
   &PartitionParticles(Number Of Partitions=16)
   /SetVariables(Particles.RightSide:NiagaraBool=&SetBoolByIntComparison(A=$Particles.PartitionParticles.Partition, B=7, Comparison Type=::NewEnumerator0)) {isSetParameters=true, module=SetVariables_#3}
   /SetVariables(Particles.MeshIndex=&ModuloInt(A=$Particles.PartitionParticles.Partition, B=8)) {isSetParameters=true, module=SetVariables_#4}
   &InitializeParticle(
    Color=<rgba 0.8125 2 2 1>
    Lifetime=7
    Mesh Scale Mode=::NewEnumerator3
    Position Offset=&Multiply_Vector(
     A=&SelectVectorFromArray(
      Direct Array Index=$Particles.PartitionParticles.Partition
      Direct Array Mode=::NewEnumerator1
      Vector Selection Array=DataInterface() {propertyValues {FloatData [<22 25 0>, <35 52 0>, <17 44 0>, <4 29 0>, <-8 16 0>, <-18 32 0>, <-30 22 0>, <-55 43 0>]}}
     )
     B=&MakeCustomVectorFromBool(Bool=$Particles.RightSide, True Vector=<1 1 1>, False Vector=<1 -1 0>)
    )
    Sprite Size Mode=::NewEnumerator3
    UsePositionOffset=true
    Write Material Random=true
   )
   /SetVariables(Particles.Distribution:NiagaraFloat=&FloatFromCurve(
     FloatCurve=DataInterface() {propertyValues {Curve {keys [<at 0 1>,<at 75 0>]}}}
     CurveIndex=&DistanceBetweenPositions(
      A=&Multiply_Vector(A=Linked() {linkedVariable {name=Particles.Position, type=NiagaraPosition}}, B=<1 0.5 1>)
      B=<25 0 0>
     )
    )) {isSetParameters=true, module=SetVariables_#5}
   &InitialMeshOrientation(
    Mesh Orientation Mode=::NewEnumerator2
    Rotation=&MakeCustomVectorFromBool(Bool=$Particles.RightSide, True Vector=<0.5 0 0>, False Vector=<0 0 0>)
    Use Rotation Vector=true
   )
   &AddVelocity(Module.Use Falloff Distance=false, Override Seed=false, Velocity Mode=::NewEnumerator1, Velocity Speed=500)
  ]
  particleUpdateScript [
   &ParticleState(Kill Particles When Lifetime Has Elapsed=false)
   &ScaleMeshSize(Scale Factor=&Lerp_Vector(A=<0.6 0.8 1>, B=<1 1 1>, Alpha=$System.FormUp))
   &self:GetButterflyTransform(ButterflyParticle=DataInterface() {propertyValues {EmitterBinding {bindingMode=Other, emitterName=ButterflyBody}}}) {module=GetButterflyTransform}
   &IncrementOverTime(Rate of Change:NiagaraFloat=&Lerp_Float(A=0.1, B=1.25, Alpha=$System.FormUp))
   /SetVariables(Particles.FlapAnim:NiagaraFloat=&FloatFromCurve(
     FloatCurve=DataInterface() {propertyValues {Curve {keys [
         {interpMode=RCIM_Cubic, tangentMode=RCTM_User, time=0, value=0}
         {arriveTangent=4.1706, interpMode=RCIM_Cubic, leaveTangent=4.1706, tangentMode=RCTM_User, time=0.25, value=0.7}
         {
          arriveTangent=-0.2246
          arriveTangentWeight=0.2294
          interpMode=RCIM_Cubic
          leaveTangent=-0.2246
          leaveTangentWeight=0.0573
          tangentMode=RCTM_User
          time=0.45
          value=1
         }
         {arriveTangent=-2.436, interpMode=RCIM_Cubic, leaveTangent=-2.436, tangentMode=RCTM_User, time=0.6, value=0.4121}
         {tangentMode=RCTM_User, time=1, value=0}
        ]}}}
     CurveIndex=&ModuloFloat(
      A=&Add_Float(A=$Particles.IncrementOverTime.Counter, B=&Lerp_Float(A=-0.08, B=0.05, Alpha=$Particles.Distribution))
      B=1
     )
    )) {isSetParameters=true}
   /SetVariables(Particles.FlapAnimVector:Vector3f=&RotateVectorByQuaternion(
     Vector To Rotate=Linked() {linkedVariable {name=Particles.Initial.Position, type=NiagaraPosition}}
     Quaternion=&MakeQuatFromAxisAngle(
      Axis=<1 0 0>
      AngleInDegrees=&Lerp_Float(
       A=-80
       B=70
       Alpha=&MakeCustomFloatFromBool(Bool=$Particles.RightSide, True Float=$Particles.FlapAnim, False Float=&OneMinusFloat(Float=$Particles.FlapAnim))
      )
     )
    )) {isSetParameters=true}
   &InterpolateOverTime(
    Rate Of Change:NiagaraFloat=&Lerp_Float(A=&Lerp_Float(A=2, B=8, Alpha=$Particles.Distribution), B=20, Alpha=$System.FormUp)
    Target Vector Value:Vector3f=&Lerp_Vector(
     A=Linked() {linkedVariable {name=Particles.Position, type=NiagaraPosition}}
     B=&AddVectorToPosition(
      Position=$Output.GetButterflyTransform.ButterflyPosition
      Vector=&RotateVectorByQuaternion(Vector To Rotate=$Particles.FlapAnimVector, Quaternion=$Output.GetButterflyTransform.ButterflyOrientation)
     )
     Alpha=$System.FormUp
    )
    Variable Type=::NewEnumerator2
   ) {module=InterpolateOverTime001}
   /SetVariables(Particles.Position=$Particles.InterpolateOverTime001.InterpolateOverTime_V3_.Moving Average) {isSetParameters=true}
   &UpdateMeshOrientation(
    Facing Interpolation=false
    Facing Position:NiagaraPosition=$Output.GetButterflyTransform.ButterflyPosition
    Orientation Method=::NewEnumerator2
    Override Quat=false
    Side Direction:Vector3f=&MakeCustomVectorFromBool(
     Bool=$Particles.RightSide
     True Vector=$Output.GetButterflyTransform.ButterflyDirection
     False Vector=&Multiply_VectorByFloat(Vector=$Output.GetButterflyTransform.ButterflyDirection, Float=-1)
    )
    Stabilization Mode:ENiagara_OrientVectorCount=::NewEnumerator2
    TransformFacingOrigin=false
    TransformFacingPosition=false
    TransformSide=false
   )
   &VortexForce(
    Origin Pull Amount:NiagaraFloat=&Lerp_Float(A=400, B=-1700, Alpha=$System.Fling)
    Origin Pull Only On Vortex Axis=false
    UseOriginOffset=false
    Vortex Force Amount=&Lerp_Float(A=350, B=0, Alpha=$System.FormUp)
    Vortex Origin:NiagaraPosition=&AddVectorToPosition(
     Position=$Output.GetButterflyTransform.ButterflyPosition
     Vector=&Multiply_VectorByFloat(Vector=$Output.GetButterflyTransform.ButterflyDirection, Float=50)
    )
   )
   &CurlNoiseForce(
    Noise Frequency=&RandomRangeFloat(Minimum=8, Maximum=11)
    Noise Strength=&Lerp_Float(A=500, B=0, Alpha=$System.FormUp)
    Pan Noise=false
   )
   &Drag(Drag=&Lerp_Float(A=2, B=10, Alpha=$System.FormUp), Use Linear Drag=true, Use Rotational Drag=false)
   &SolveForcesAndVelocity(Clamp Velocity=false, Limit Acceleration=false)
   &InterpolateOverTime(
    Rate Of Change=2
    Target Quat Value:Quat4f=&FindQuatBetweenTwoVectors(From Vector=&RandomVector(Vector Scale=1), To Vector=$Output.CurlNoiseForce.CurlNoiseForce)
    Variable Type=::NewEnumerator5
   )
   /SetVariables(Particles.OrientationAnim:Quat4f=&LerpQuaternion(
     Quaternion A:Quat4f=$Particles.InterpolateOverTime.InterpolateOverTime_Quaternion.Moving Average
     Quaternion B:Quat4f=$Particles.MeshOrientation
     Lerp Factor:NiagaraFloat=&FloatFromCurve(
      FloatCurve=DataInterface() {propertyValues {Curve {keys [<at 0 0 RCIM_Cubic>, {tangentMode=RCTM_User, time=0.4, value=1}]}}}
      CurveIndex=$System.FormUp
     )
     Enable Spherical Linear Interpolation:NiagaraBool=false
    )) {isSetParameters=true}
   &ScaleColor(
    Scale Alpha=$User.Fade
    Scale RGB=&Lerp_Vector(A=<0.5 0.5 0.5>, B=<1 1 1>, Alpha=$System.FormUp)
    ScaleA=true
    ScaleRGB=true
   )
  ]
  rendererClasses [&NiagaraMeshRendererProperties, &NiagaraSpriteRendererProperties]
  renderers [
   {rendererClass=&NiagaraMeshRendererProperties, rendererIndex=0}
   {rendererClass=&NiagaraSpriteRendererProperties, rendererIndex=1}
  ]
 }
 DriftingShards {
  emitterUpdateScript [
   &EmitterState()
   &SpawnRate(SpawnRate=6, Use Spawn Probability=false)
  ]
  particleSpawnScript [
   &PartitionParticles(Number Of Partitions=4)
   &InitializeParticle(
    Color=<rgba 1.3868 2.4194 3 1>
    Lifetime Max=4
    Lifetime Min=2
    Lifetime Mode=::NewEnumerator1
    Mesh Scale Max=<0.8 1 2>
    Mesh Scale Min=<0.5 0.6 2>
    Mesh Scale Mode=::NewEnumerator2
    UsePositionOffset=false
    Write Material Random=true
   )
   /SetVariables(Particles.MeshIndex=$Particles.PartitionParticles.Partition) {isSetParameters=true}
   &ShapeLocation(
    MaskSpawnGroup=false
    Non Uniform Scale=<1 1 0.8>
    Override Seed=false
    Sphere Radius=200
    Sphere Surface Distribution=0.5
    UseOwnerScale=true
   )
   &InitialMeshOrientation(Use Rotation Vector=false)
   &AddVelocity(Module.Use Falloff Distance=false, Override Seed=false, Velocity Mode=::NewEnumerator1, Velocity Speed=100)
  ]
  particleUpdateScript [
   &ParticleState()
   &ScaleMeshSize(Scale Factor=&VectorFromFloat(Value=&FloatFromCurve(FloatCurve=DataInterface() {propertyValues {Curve {keys [
          {arriveTangent=7.2427, interpMode=RCIM_Cubic, leaveTangent=7.2427, tangentMode=RCTM_User, time=0, value=0}
          {arriveTangentWeight=0.1491, interpMode=RCIM_Cubic, leaveTangentWeight=0.2236, tangentMode=RCTM_User, time=0.2, value=1}
          {arriveTangentWeight=0.1, interpMode=RCIM_Cubic, leaveTangentWeight=0.1667, tangentMode=RCTM_User, time=0.8, value=1}
          {arriveTangent=-8.0007, tangentMode=RCTM_User, time=1, value=0}
         ]}}})))
   &UpdateMeshOrientation(
    Rotation Rate=&RandomRangeFloat(Minimum=0.2, Maximum=0.5)
    Rotation Vector=&RandomVector(Vector Scale=1)
   )
   &VortexForce(Origin Pull Amount=70, Origin Pull Only On Vortex Axis=false, UseOriginOffset=false)
   &CurlNoiseForce(Noise Frequency=&RandomRangeFloat(Minimum=4, Maximum=8), Noise Strength=300, Pan Noise=false)
   &Drag(Drag=&RandomRangeFloat(Minimum=4, Maximum=6), Use Linear Drag=true, Use Rotational Drag=false)
   &SolveForcesAndVelocity(Clamp Velocity=false, Limit Acceleration=false)
  ]
  rendererClasses [&NiagaraMeshRendererProperties]
  renderers [{rendererClass=&NiagaraMeshRendererProperties, rendererIndex=0}]
 }
 Layout {
  bEnabled=false
  emitterUpdateScript [
   &EmitterState(
    Inactive Response=::NewEnumerator1
    Life Cycle Mode=::NewEnumerator1
    Loop Behavior=::NewEnumerator1
    Loop Duration Mode=::NewEnumerator1
    UseLoopDelay=false
   )
   &SpawnRate(Use Spawn Probability=false) {enabled=false}
   &SpawnBurst_Instantaneous(Spawn Count=16, Use Spawn Probability=false, UseLoopCountLimit=false)
  ]
  particleSpawnScript [
   &PartitionParticles(Number Of Partitions=16)
   /SetVariables(Particles.RightSide:NiagaraBool=&SetBoolByIntComparison(A=$Particles.PartitionParticles.Partition, B=7, Comparison Type=::NewEnumerator0)) {isSetParameters=true, module=SetVariables_#3}
   /SetVariables(Particles.Distribution:NiagaraFloat=&FloatFromCurve(
     FloatCurve=DataInterface() {propertyValues {Curve {keys [<at 0 0>,<at 7 1>]}}}
     CurveIndex=&MakeFloatFromInt(Int=&ModuloInt(A=$Particles.PartitionParticles.Partition, B=8))
    )) {enabled=false, isSetParameters=true, module=SetVariables_#5}
   /SetVariables(Particles.MeshIndex=&ModuloInt(A=$Particles.PartitionParticles.Partition, B=8)) {isSetParameters=true, module=SetVariables_#4}
   &InitializeParticle(
    Color=<rgba 1.4531 2 1.8383 1>
    Lifetime=7
    Mesh Scale Mode=::NewEnumerator3
    Position Offset=&Multiply_Vector(
     A=&SelectVectorFromArray(
      Direct Array Index=$Particles.PartitionParticles.Partition
      Direct Array Mode=::NewEnumerator1
      Vector Selection Array=DataInterface() {propertyValues {FloatData [<22 25 0>, <35 52 0>, <17 44 0>, <4 29 0>, <-8 16 0>, <-18 32 0>, <-30 22 0>, <-55 43 0>]}}
     )
     B=&MakeCustomVectorFromBool(Bool=$Particles.RightSide, True Vector=<1 1 1>, False Vector=<1 -1 0>)
    )
    Sprite Size Mode=::NewEnumerator3
    UsePositionOffset=true
    Write Material Random=true
   )
   &InitialMeshOrientation(
    Mesh Orientation Mode=::NewEnumerator2
    Orientation Vector=&Multiply_VectorByFloat(Vector=Linked() {linkedVariable {name=Particles.Position, type=NiagaraPosition}}, Float=-1)
    Rotation=&MakeCustomVectorFromBool(Bool=$Particles.RightSide, True Vector=<0.5 0 0>, False Vector=<0 0 0>)
    Use Rotation Vector=true
   )
  ]
  particleUpdateScript [&ParticleState(Kill Particles When Lifetime Has Elapsed=false)]
  rendererClasses [&NiagaraMeshRendererProperties, &NiagaraSpriteRendererProperties]
  renderers [
   {rendererClass=&NiagaraMeshRendererProperties, rendererIndex=0}
   {rendererClass=&NiagaraSpriteRendererProperties, rendererIndex=1}
  ]
 }
 Sparkles {
  emitterUpdateScript [
   &EmitterState(Enable Distance Culling=true, Module.bUseMaxDistance=true, Module.bUseMinDistance=false, Scalability Mode=::NewEnumerator1)
   &SpawnParticlesFromOtherEmitter(
    Attribute Reader=DataInterface() {propertyValues {EmitterBinding {bindingMode=Other, emitterName=ButterflyBody}}}
    Spawn Rate=&Lerp_Float(A=0, B=15, Alpha=&Add_Float(A=$System.FormUp, B=$System.Fling))
    Use Maximum Age=false
    Use Minimum Age=false
   )
  ]
  particleSpawnScript [
   /SetVariables(Particles.Distribution:NiagaraFloat=&FloatFromCurve(
     FloatCurve=DataInterface() {propertyValues {Curve {keys [
         {interpMode=RCIM_Cubic, leaveTangent=0.4364, tangentMode=RCTM_Break, time=0, value=0}
         {arriveTangent=1.9884, interpMode=RCIM_Cubic, leaveTangent=1.9884, tangentMode=RCTM_User, time=0.64, value=0.5}
         <at 1 1 RCIM_Cubic>
        ]}}}
     CurveIndex=&RandomRangeFloat(Minimum=0, Maximum=1)
    )) {isSetParameters=true}
   &InitializeParticle(
    Color=&LerpLinearColors(StartColor=<rgba 0.0469 0.2977 3 1>, EndColor=<rgba 0.0625 1.5073 4 1>, LerpFactor=$Particles.Distribution)
    Lifetime=&RandomRangeFloat(Minimum=1, Maximum=3)
    Sprite Size Mode=::NewEnumerator3
    Uniform Sprite Size=&Lerp_Float(A=3, B=4.5, Alpha=$Particles.Distribution)
    UsePositionOffset=false
    Write Material Random=true
   )
   &SampleParticlesFromOtherEmitter(
    Applied Velocity Scale:NiagaraFloat=&RandomRangeFloat(Minimum=0.2, Maximum=0.6)
    Velocity Sampling=::NewEnumerator0
   )
   &ShapeLocation(
    MaskSpawnGroup=false
    Override Seed=false
    Shape Origin=$Output.SampleParticlesFromOtherEmitter.SampledPosition
    Sphere Radius=3
    Sphere Surface Distribution=1
    UseOwnerScale=true
   )
  ]
  particleUpdateScript [
   &ParticleState()
   &DynamicMaterialParameters(
    Index 0 Param 1=&FloatFromCurve(FloatCurve=DataInterface() {propertyValues {Curve {keys [<at 0.1 1 RCIM_Cubic>,<at 1 4>]}}})
    Index 0 Param 2=3
    Index 0 Param 3=3
    Index 0 Param 4=1
    Module.Param0WriteEnabled=true
    Module.Param1WriteEnabled=true
    Module.Param2WriteEnabled=true
    Module.Param3WriteEnabled=true
   )
   &VectorNoiseForce(Force Amount=&Lerp_Float(A=50, B=500, Alpha=$System.Fling), Module.Use Random Noise Time Interval=false)
   &Drag(Drag=0.5, Use Linear Drag=true, Use Rotational Drag=false)
   &SolveForcesAndVelocity(Clamp Velocity=false, Limit Acceleration=false)
   &ScaleSpriteSize(Uniform Curve Sprite Scale=DataInterface() {propertyValues {Curve {keys [<at 0.5 1 RCIM_Cubic>, {arriveTangent=-4.6267, interpMode=RCIM_Cubic, leaveTangent=-4.6267, tangentMode=RCTM_User, time=1, value=0}]}}})
   &ScaleSpriteSizeBySpeed(
    Max Scale Factor=<0.6 4>
    Sample Scale Factor By Curve=true
    Scale Factor Curve=DataInterface() {propertyValues {Curve {keys [{arriveTangent=0.5646, interpMode=RCIM_Cubic, time=0, value=0}, {interpMode=RCIM_Cubic, leaveTangent=2.0784, time=1, value=1}]}}}
    Velocity Threshold=170
   )
   &ScaleColor(
    Scale Alpha=&FloatFromCurve(FloatCurve=DataInterface() {propertyValues {Curve {keys [
         {arriveTangent=13.8764, interpMode=RCIM_Cubic, leaveTangent=13.8764, tangentMode=RCTM_User, time=0, value=0}
         {arriveTangentWeight=0.1163, interpMode=RCIM_Cubic, leaveTangentWeight=0.2906, tangentMode=RCTM_User, time=0.1, value=1}
         {arriveTangentWeight=0.1667, interpMode=RCIM_Cubic, leaveTangentWeight=0.1, tangentMode=RCTM_User, time=0.7, value=1}
         {tangentMode=RCTM_User, time=1, value=0}
        ]}}})
    ScaleA=true
    ScaleRGB=false
   )
  ]
  rendererClasses [&NiagaraSpriteRendererProperties]
  renderers [{rendererClass=&NiagaraSpriteRendererProperties}]
 }
}
summary {
 systemName=NS_KelpLinen_DJ_Butterfly
 emitters [
  {emitterName=ButterflyBody, rendererClasses [&NiagaraSpriteRendererProperties, &NiagaraMeshRendererProperties]}
  {
   emitterName=ButterflyShards
   rendererClasses [&NiagaraMeshRendererProperties, &NiagaraSpriteRendererProperties]
  }
  {
   bEnabled=false
   emitterName=Layout
   rendererClasses [&NiagaraMeshRendererProperties, &NiagaraSpriteRendererProperties]
  }
  {emitterName=DriftingShards, rendererClasses [&NiagaraMeshRendererProperties]}
  {emitterName=Sparkles, rendererClasses [&NiagaraSpriteRendererProperties]}
 ]
 userVariables [User.Fade {type=NiagaraFloat, defaultValue=1, description=""}]
}
